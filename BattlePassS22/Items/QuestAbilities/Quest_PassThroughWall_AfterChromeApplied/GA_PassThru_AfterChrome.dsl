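; Quest gameplay ability that credits phasing through a wall shortly after a chrome splash.
; On event activation it flags a recent splash, waits for the phased-through-wall tag, and starts a looping timer with a randomized interval.
; If the tag arrives while flagged, it sends a PhaseAfterSplash stat event to the quest manager.
; Timer ticks accumulate elapsed time until Max Duration clears the flag and ends the ability, which clears the timer.
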(event Ability|EventActivateAbilityFromEvent
  (params (EventData :type Gameplay%20Event%20Data%20Structure%20%28by%20ref%29))
  ;[1 1 1] Activate ability - listen for when a player throws chrome splash, then start a timer while waiting for a phase event
  (|CallParentFunction EventData)
  (Variables|Default|SetRecentlySplashed true)
  (Ability|Tasks|WaitGameplayTagAdd (Variables|Default|GetPhasedThroughWall)
    (:then
      (bind _returnvalue (Utilities|Time|SetTimerbyEvent (delegate Custom|WaitForPhase) (Math|Random|RandomFloatinRange 0.5 0.8) true))
      (Variables|Default|SetTimer%20Handle _returnvalue))
    (:Added
      (if (Variables|Default|GetRecentlySplashed)
        (Quest|SendComplexCustomStatEvent (Variables|GAVars|Data|GetOwningPlayerQuestManager) 0 "(GameplayTags=((TagName=\"Athena.Quests.PhaseAfterSplash\")))" "(GameplayTags=)" 1)))))

(event Custom|WaitForPhase
  (bind _timer_handle (Variables|Default|GetTimer%20Handle))
  (bind _returnvalue (Utilities|Time|GetTimerElapsedTimebyHandle _timer_handle))
  (bind _time_elapsed (Variables|Default|GetTime%20Elapsed))
  (bind _returnvalue_2 (+ _returnvalue _time_elapsed))
  ;[1 1 1] Activate ability - listen for when a player throws chrome splash, then start a timer while waiting for a phase event
  (Variables|Default|SetTime%20Elapsed _returnvalue_2)
  (if (>= _returnvalue_2 (Variables|Default|GetMax%20Duration))
    (Variables|Default|SetRecentlySplashed false)
    (Ability|EndAbility)))

(event Ability|EventOnEndAbility
  (params (bWasCancelled :type Boolean))
  (bind _timer_handle (Variables|Default|GetTimer%20Handle))
  ;[1 1 1] Cleanup
  (if (Utilities|Time|IsValidTimerHandle _timer_handle)
    (Utilities|Time|ClearandInvalidateTimerbyHandle _timer_handle)))
